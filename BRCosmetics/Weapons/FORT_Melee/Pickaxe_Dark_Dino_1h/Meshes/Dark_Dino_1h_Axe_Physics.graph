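# Physics asset for the Dark Dino one-handed pickaxe, providing box bodies for the axe and its dangling ribbons.
# One box covers the Melee_TwoHanded_Handle bone, and smaller boxes cover a master ribbon bone and two ribbon pieces.
# The ribbon master is constrained to the handle with free swing and twist.
# Both ribbon pieces branch from the master with wide swing limits and tighter twist limits.
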PhysicsAsset /BRCosmetics/Weapons/FORT_Melee/Pickaxe_Dark_Dino_1h/Meshes/Dark_Dino_1h_Axe_Physics
bodies {
 Melee_TwoHanded_Handle [{shapeType=Box, extentX=7.1812, extentY=14.1816, extentZ=69.388, center=<-0.5486 -8.8625 46.6085>}]
 dyn_ribbon_a [{shapeType=Box, extentX=6.2817, extentY=1.2804, extentZ=4.1156, center=<4.4953 -0.0774 0.6596>}]
 dyn_ribbon_b [{shapeType=Box, extentX=4.4848, extentY=1.0539, extentZ=2.814, center=<4.6708 -0.0276 -0.2531>}]
 dyn_ribbon_mstr [{shapeType=Box, extentX=1.01, extentY=1.01, extentZ=1.01}]
}
constraints [
 {
  bone1Name=dyn_ribbon_mstr
  bone2Name=Melee_TwoHanded_Handle
  swing1LimitDegrees=45
  swing1Motion=Free
  swing2LimitDegrees=45
  swing2Motion=Free
  twistLimitDegrees=45
  twistMotion=Free
 }
 {bone1Name=dyn_ribbon_a, bone2Name=dyn_ribbon_mstr, swing1LimitDegrees=75, swing2LimitDegrees=75, twistLimitDegrees=25}
 {bone1Name=dyn_ribbon_b, bone2Name=dyn_ribbon_mstr, swing1LimitDegrees=75, swing2LimitDegrees=75, twistLimitDegrees=25}
]
